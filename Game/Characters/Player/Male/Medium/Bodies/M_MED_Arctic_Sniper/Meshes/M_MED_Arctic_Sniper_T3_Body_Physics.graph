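# Physics asset for the male medium Arctic Sniper T3 player body, covering only the shoulder fur bones for secondary motion.
# Small capsules on spine_05, both clavicles and front and back master bones form the base, with clavicles and masters joined to spine_05 by free-rotating constraints.
# Larger fur capsules hang from these, three per clavicle and one per master, under narrower swing and twist limits.
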PhysicsAsset /Game/Characters/Player/Male/Medium/Bodies/M_MED_Arctic_Sniper/Meshes/M_MED_Arctic_Sniper_T3_Body_Physics
bodies {
 clavicle_l [{shapeType=Capsule, length=0, radius=2}]
 clavicle_r [{shapeType=Capsule, length=0, radius=2}]
 dyn_shouderFur_back [{shapeType=Capsule, length=2, radius=5, center=<0 -1.6037 0>}]
 dyn_shouderFur_back_master [{shapeType=Capsule, length=0.505, radius=0.505, rotation=<rot 0 -90 0>}]
 dyn_shouderFur_front [{shapeType=Capsule, length=2, radius=5, center=<0 2.1881 0>}]
 dyn_shouderFur_front_master [{shapeType=Capsule, length=0.505, radius=0.505, rotation=<rot 0 -90 0>}]
 dyn_shoulderFur_01_l [{shapeType=Capsule, length=2, radius=5, center=<0 0 3>}]
 dyn_shoulderFur_01_r [{shapeType=Capsule, length=2, radius=5, center=<0 0 -3>}]
 dyn_shoulderFur_02_l [{shapeType=Capsule, length=2, radius=5, center=<0 0 3>}]
 dyn_shoulderFur_02_r [{shapeType=Capsule, length=2, radius=5, center=<0 0 -3>}]
 dyn_shoulderFur_03_l [{shapeType=Capsule, length=2, radius=5, center=<0 0 3>}]
 dyn_shoulderFur_03_r [{shapeType=Capsule, length=2, radius=5, center=<0 -2.0228 0>}]
 spine_05 [{shapeType=Capsule, length=0, radius=2}]
}
constraints [
 {
  bone1Name=clavicle_l
  bone2Name=spine_05
  swing1LimitDegrees=45
  swing1Motion=Free
  swing2LimitDegrees=45
  swing2Motion=Free
  twistLimitDegrees=45
  twistMotion=Free
 }
 {
  bone1Name=clavicle_r
  bone2Name=spine_05
  swing1LimitDegrees=45
  swing1Motion=Free
  swing2LimitDegrees=45
  swing2Motion=Free
  twistLimitDegrees=45
  twistMotion=Free
 }
 {bone1Name=dyn_shoulderFur_01_l, bone2Name=clavicle_l, swing1LimitDegrees=30, swing2LimitDegrees=30, twistLimitDegrees=10}
 {bone1Name=dyn_shoulderFur_02_l, bone2Name=clavicle_l, swing1LimitDegrees=30, swing2LimitDegrees=30, twistLimitDegrees=10}
 {bone1Name=dyn_shoulderFur_03_l, bone2Name=clavicle_l, swing1LimitDegrees=30, swing2LimitDegrees=30, twistLimitDegrees=10}
 {bone1Name=dyn_shoulderFur_01_r, bone2Name=clavicle_r, swing1LimitDegrees=30, swing2LimitDegrees=30, twistLimitDegrees=10}
 {bone1Name=dyn_shoulderFur_02_r, bone2Name=clavicle_r, swing1LimitDegrees=30, swing2LimitDegrees=30, twistLimitDegrees=10}
 {bone1Name=dyn_shoulderFur_03_r, bone2Name=clavicle_r, swing1LimitDegrees=30, swing2LimitDegrees=30, twistLimitDegrees=10}
 {bone1Name=dyn_shouderFur_back, bone2Name=dyn_shouderFur_back_master, swing1LimitDegrees=30, swing2LimitDegrees=30, twistLimitDegrees=10}
 {bone1Name=dyn_shouderFur_front, bone2Name=dyn_shouderFur_front_master, swing1LimitDegrees=30, swing2LimitDegrees=30, twistLimitDegrees=5}
 {
  bone1Name=dyn_shouderFur_back_master
  bone2Name=spine_05
  swing1LimitDegrees=45
  swing1Motion=Free
  swing2LimitDegrees=45
  swing2Motion=Free
  twistLimitDegrees=45
  twistMotion=Free
 }
 {
  bone1Name=dyn_shouderFur_front_master
  bone2Name=spine_05
  swing1LimitDegrees=45
  swing1Motion=Free
  swing2LimitDegrees=45
  swing2Motion=Free
  twistLimitDegrees=45
  twistMotion=Free
 }
]
