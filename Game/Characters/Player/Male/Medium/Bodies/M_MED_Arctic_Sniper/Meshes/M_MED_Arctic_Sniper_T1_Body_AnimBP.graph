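# Anim Blueprint for the male medium Arctic Sniper T1 body that copies the attached parent's pose and adds secondary motion to hip straps and shoulder fur.
# Chained AnimDynamics springs drive both three-bone hip straps, with thigh-twist spherical limits on the root segments.
# Weighted constraints position the fur master bones before a RigidBody sim in spine_05 space, which borrows the T3 body physics asset.
# Additive offsets on left strap and fur bones apply only in the frontend, via Frontend_Alpha.
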AnimBlueprint /Game/Characters/Player/Male/Medium/Bodies/M_MED_Arctic_Sniper/Meshes/M_MED_Arctic_Sniper_T1_Body_AnimBP
target_skeleton=&/M_MED_Arctic_Sniper_Body_Skeleton
graphs [
 AnimGraph {nodes [
   n1 = Anim Dynamics - Bone: dyn_hipStrap_01_l(ComponentPose=Local To Component(LocalPose=Copy Pose From Mesh() {bUseAttachedParent=true}), Alpha=1, ExternalForce="0,0,0") {
    angularDampingOverride=0.9999
    angularSpringConstant=1200
    bAngularSpring=true
    bGravityOverrideInSimSpace=true
    bOverrideAngularDamping=true
    bUseSphericalLimits=true
    gravityScale=0.35
    simulationSpace=RootRelative
    alphaBoolBlend {blendOption=Linear}
    boundBone {boneName=dyn_hipStrap_01_l}
    componentAppliedLinearAccClamp=<10 10 10>
    componentLinearAccScale=<1 1 1>
    componentLinearVelScale=<1 1 1>
    physicsBodyDefinitions [{
      collisionType=CustomSphere
      sphereCollisionRadius=2
      boundBone {boneName=dyn_hipStrap_01_l}
      boxExtents=<5 20 20>
      constraintSetup {angularLimitsMax=<0 10 20>, angularLimitsMin=<0 -20 -20>, angularTarget=<1 0 0>}
      localJointOffset=<5 0 0>
     }]
    sphericalLimits [{limitRadius=60, drivingBone {boneName=thigh_twist_01_l}, sphereLocalOffset=<0.5 1 51>}]
   }
   n2 = Anim Dynamics - Bone: dyn_hipStrap_02_l(ComponentPose=@n1, Alpha=1, ExternalForce="0,0,0") {
    angularDampingOverride=0.9999
    angularSpringConstant=1200
    bAngularSpring=true
    bGravityOverrideInSimSpace=true
    bOverrideAngularDamping=true
    gravityScale=0.35
    simulationSpace=RootRelative
    alphaBoolBlend {blendOption=Linear}
    boundBone {boneName=dyn_hipStrap_02_l}
    componentAppliedLinearAccClamp=<10 10 10>
    componentLinearAccScale=<1 1 1>
    componentLinearVelScale=<1 1 1>
    physicsBodyDefinitions [{
      sphereCollisionRadius=1
      boundBone {boneName=dyn_hipStrap_02_l}
      boxExtents=<5 20 20>
      constraintSetup {angularLimitsMax=<0 2 20>, angularLimitsMin=<0 -20 -20>, angularTarget=<0 0 0>}
      localJointOffset=<5 0 0>
     }]
   }
   n3 = Anim Dynamics - Bone: dyn_hipStrap_03_l(ComponentPose=@n2, Alpha=1, ExternalForce="0,0,0") {
    angularDampingOverride=0.9999
    angularSpringConstant=1200
    bAngularSpring=true
    bGravityOverrideInSimSpace=true
    bOverrideAngularDamping=true
    gravityScale=0.35
    simulationSpace=RootRelative
    alphaBoolBlend {blendOption=Linear}
    boundBone {boneName=dyn_hipStrap_03_l}
    componentAppliedLinearAccClamp=<10 10 10>
    componentLinearAccScale=<1 1 1>
    componentLinearVelScale=<1 1 1>
    physicsBodyDefinitions [{
      sphereCollisionRadius=0
      boundBone {boneName=dyn_hipStrap_03_l}
      boxExtents=<5 20 20>
      constraintSetup {angularLimitsMax=<5 20 5>, angularLimitsMin=<-5 -20 -5>, angularTarget=<0 0 0>}
      localJointOffset=<5 0 0>
     }]
   }
   n4 = Anim Dynamics - Bone: dyn_hipStrap_01_r(ComponentPose=@n3, Alpha=1, ExternalForce="0,0,0") {
    angularDampingOverride=0.9999
    angularSpringConstant=1200
    bAngularSpring=true
    bGravityOverrideInSimSpace=true
    bOverrideAngularDamping=true
    bUseSphericalLimits=true
    gravityScale=0.35
    simulationSpace=RootRelative
    alphaBoolBlend {blendOption=Linear}
    boundBone {boneName=dyn_hipStrap_01_r}
    componentAppliedLinearAccClamp=<10 10 10>
    componentLinearAccScale=<1 1 1>
    componentLinearVelScale=<1 1 1>
    physicsBodyDefinitions [{
      collisionType=CustomSphere
      sphereCollisionRadius=2
      boundBone {boneName=dyn_hipStrap_01_r}
      boxExtents=<5 20 20>
      constraintSetup {angularLimitsMax=<0 20 15>, angularLimitsMin=<0 -10 -15>, angularTarget=<0 0 0>}
      localJointOffset=<5 0 0>
     }]
    sphericalLimits [{limitRadius=60, drivingBone {boneName=thigh_twist_01_r}, sphereLocalOffset=<-0.5 -1 -51>}]
   }
   n5 = Anim Dynamics - Bone: dyn_hipStrap_02_r(ComponentPose=@n4, Alpha=1, ExternalForce="0,0,0") {
    angularDampingOverride=0.9999
    angularSpringConstant=1200
    bAngularSpring=true
    bGravityOverrideInSimSpace=true
    bOverrideAngularDamping=true
    gravityScale=0.35
    simulationSpace=RootRelative
    alphaBoolBlend {blendOption=Linear}
    boundBone {boneName=dyn_hipStrap_02_r}
    componentAppliedLinearAccClamp=<10 10 10>
    componentLinearAccScale=<1 1 1>
    componentLinearVelScale=<1 1 1>
    physicsBodyDefinitions [{
      sphereCollisionRadius=0
      boundBone {boneName=dyn_hipStrap_02_r}
      boxExtents=<5 20 20>
      constraintSetup {angularLimitsMax=<0 20 20>, angularLimitsMin=<0 -1 -20>, angularTarget=<0 0 0>}
      localJointOffset=<5 0 0>
     }]
   }
   n6 = Anim Dynamics - Bone: dyn_hipStrap_03_r(ComponentPose=@n5, Alpha=1, ExternalForce="0,0,0") {
    angularDampingOverride=0.9999
    angularSpringConstant=1200
    bAngularSpring=true
    bGravityOverrideInSimSpace=true
    bOverrideAngularDamping=true
    gravityScale=0.35
    simulationSpace=RootRelative
    alphaBoolBlend {blendOption=Linear}
    boundBone {boneName=dyn_hipStrap_03_r}
    componentAppliedLinearAccClamp=<10 10 10>
    componentLinearAccScale=<1 1 1>
    componentLinearVelScale=<1 1 1>
    physicsBodyDefinitions [{
      sphereCollisionRadius=0
      boundBone {boneName=dyn_hipStrap_03_r}
      boxExtents=<5 20 20>
      constraintSetup {angularLimitsMax=<5 20 5>, angularLimitsMin=<-5 -20 -5>, angularTarget=<0 0 0>}
      localJointOffset=<5 0 0>
     }]
   }
   n7 = "Transform (Modify) Bone - Bone: dyn_hipStrap_01_l"(ComponentPose=@n6, Alpha=Get Frontend_Alpha() -> (Frontend_Alpha:double), Translation="0,0,0", Rotation="3,0,0", Scale="1,1,1") {
    rotationMode=BMM_Additive
    rotationSpace=BCS_ParentBoneSpace
    translationMode=BMM_Additive
    translationSpace=BCS_ParentBoneSpace
    alphaBoolBlend {blendOption=Linear}
    boneToModify {boneName=dyn_hipStrap_01_l}
    rotation {pitch=3}
   }
   n8 = "Constraint - dyn_shouderFur_front_master (PPPPPPPP)"!(
    ComponentPose=@n7
    Alpha=1
    ConstraintWeights_0=0.2784
    ConstraintWeights_1:float=0
    ConstraintWeights_2:float=0.0091
    ConstraintWeights_3:float=0.151
    ConstraintWeights_4:float=0.4398
    ConstraintWeights_5:float=0
    ConstraintWeights_6:float=0.0001
    ConstraintWeights_7:float=0.1216
   ) -> (Pose) {
    alphaBoolBlend {blendOption=Linear}
    boneToModify {boneName=dyn_shouderFur_front_master}
    constraintSetup [
     {targetBone {boneName=clavicle_l}}
     {targetBone {boneName=spine_01}}
     {targetBone {boneName=spine_02}}
     {targetBone {boneName=spine_03}}
     {targetBone {boneName=spine_04}}
     {targetBone {boneName=upperarm_twist_01_l}}
     {targetBone {boneName=clavicle_r}}
     {targetBone {boneName=dyn_shouderFur_front}}
    ]
    constraintWeights [1,1,1,1,1,1,1,1]
   }
   n9 = "Constraint - dyn_shouderFur_back_master (PPPPPPPPPPPP)"!(
    ComponentPose=@n8
    Alpha=1
    ConstraintWeights_0=0
    ConstraintWeights_1:float=0.2251
    ConstraintWeights_2:float=0.1825
    ConstraintWeights_3:float=0
    ConstraintWeights_4:float=0.2142
    ConstraintWeights_5:float=0.0313
    ConstraintWeights_6:float=0.0805
    ConstraintWeights_7:float=0
    ConstraintWeights_8:float=0.2614
    ConstraintWeights_9:float=0.0004
    ConstraintWeights_10:float=0.0045
    ConstraintWeights_11:float=0
   ) -> (Pose) {
    alphaBoolBlend {blendOption=Linear}
    boneToModify {boneName=dyn_shouderFur_back_master}
    constraintSetup [
     {targetBone {boneName=upperarm_r}}
     {targetBone {boneName=clavicle_l}}
     {targetBone {boneName=dyn_shouderFur_back}}
     {targetBone {boneName=spine_02}}
     {targetBone {boneName=spine_03}}
     {targetBone {boneName=spine_04}}
     {targetBone {boneName=spine_05}}
     {targetBone {boneName=upperarm_twist_01_l}}
     {targetBone {boneName=clavicle_r}}
     {targetBone {boneName=neck_02}}
     {targetBone {boneName=neck_01}}
     {targetBone {boneName=dyn_hipStrap_01_r}}
    ]
    constraintWeights [1,1,1,1,1,1,1,1,1,1,1,1]
   }
   n10 = RigidBody(ComponentPose=@n9) {
    bOverrideWorldGravity=true
    overridePhysicsAsset=&/M_MED_Arctic_Sniper_T3_Body_Physics
    simulationSpace=BaseBoneSpace
    alphaBoolBlend {blendOption=Linear}
    baseBoneRef {boneName=spine_05}
    componentAppliedLinearAccClamp=<2000 2000 2000>
    componentLinearAccScale=<10 10 10>
    componentLinearVelScale=<2 2 2>
    overrideWorldGravity {z=-980}
   }
   n11 = "Transform (Modify) Bone - Bone: dyn_shoulderFur_02_l"(ComponentPose=@n10, Alpha=Get Frontend_Alpha(), Translation="0,-0.000,1", Rotation="10,0,0", Scale="1,1,1") {
    rotationMode=BMM_Additive
    rotationSpace=BCS_ParentBoneSpace
    translationMode=BMM_Additive
    translationSpace=BCS_ParentBoneSpace
    alphaBoolBlend {blendOption=Linear}
    boneToModify {boneName=dyn_shoulderFur_02_l}
    rotation {pitch=10}
    translation {z=1}
   }
   Output Pose(Result=Component To Local(ComponentPose=@n11)) {layerGroup=None}
  ]}
 EventGraph {nodes [
   Event BlueprintUpdateAnimation()
   > Set Frontend_Alpha(Frontend_Alpha:double=SelectFloat(A=1, B=0, bPickA=Get bIsInFrontEnd())) -> (Output_Get:double)
   TryGetPawnOwner()
  ]}
]
variables [Frontend_Alpha {type=number}]
