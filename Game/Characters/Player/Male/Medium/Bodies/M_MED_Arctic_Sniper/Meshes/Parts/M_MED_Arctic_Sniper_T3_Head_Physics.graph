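# Physics asset for the Arctic Sniper T3 head part (male medium player body), setting up secondary motion for the hood fur.
# Six dyn_hoodFur bones each carry an identical small offset capsule, alongside a zero-length, effectively spherical capsule on the head.
# Each fur body is constrained directly to the head rather than chained, with moderate swing limits and tight twist limits.
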PhysicsAsset /Game/Characters/Player/Male/Medium/Bodies/M_MED_Arctic_Sniper/Meshes/Parts/M_MED_Arctic_Sniper_T3_Head_Physics
bodies {
 dyn_hoodFur_01 [{shapeType=Capsule, length=1, radius=3, center=<-3 0 0>, rotation=<rot 90 0 0>}]
 dyn_hoodFur_02 [{shapeType=Capsule, length=1, radius=3, center=<-3 0 0>, rotation=<rot 90 0 0>}]
 dyn_hoodFur_03 [{shapeType=Capsule, length=1, radius=3, center=<-3 0 0>, rotation=<rot 90 0 0>}]
 dyn_hoodFur_04 [{shapeType=Capsule, length=1, radius=3, center=<-3 0 0>, rotation=<rot 90 0 0>}]
 dyn_hoodFur_05 [{shapeType=Capsule, length=1, radius=3, center=<-3 0 0>, rotation=<rot 90 0 0>}]
 dyn_hoodFur_06 [{shapeType=Capsule, length=1, radius=3, center=<-3 0 0>, rotation=<rot 90 0 0>}]
 head [{shapeType=Capsule, length=0, radius=5}]
}
constraints [
 {bone1Name=dyn_hoodFur_06, bone2Name=head, swing1LimitDegrees=30, swing2LimitDegrees=30, twistLimitDegrees=5}
 {bone1Name=dyn_hoodFur_05, bone2Name=head, swing1LimitDegrees=30, swing2LimitDegrees=30, twistLimitDegrees=5}
 {bone1Name=dyn_hoodFur_04, bone2Name=head, swing1LimitDegrees=30, swing2LimitDegrees=30, twistLimitDegrees=5}
 {bone1Name=dyn_hoodFur_03, bone2Name=head, swing1LimitDegrees=30, swing2LimitDegrees=30, twistLimitDegrees=5}
 {bone1Name=dyn_hoodFur_02, bone2Name=head, swing1LimitDegrees=30, swing2LimitDegrees=30, twistLimitDegrees=5}
 {bone1Name=dyn_hoodFur_01, bone2Name=head, swing1LimitDegrees=30, swing2LimitDegrees=30, twistLimitDegrees=5}
]
